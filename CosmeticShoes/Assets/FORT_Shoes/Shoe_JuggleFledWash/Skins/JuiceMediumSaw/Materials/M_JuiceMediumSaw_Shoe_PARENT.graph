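# Parent material for the JuiceMediumSaw cosmetic shoe skin, layering "Sweet" and "Slurp" effect functions over the shared character and shoe-reactivity base.
# A tech-art mask drives lace hue shift and desaturation on the AO-darkened Sweet emissive and blends in the Slurp layer.
# Fresnel and sky-light-to-camera alignment modulate the blended emissive, then sock, character PostFX and hologram functions finish the chain.
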Material /CosmeticShoes/Assets/FORT_Shoes/Shoe_JuggleFledWash/Skins/JuiceMediumSaw/Materials/M_JuiceMediumSaw_Shoe_PARENT
graph {
 properties {
  bUseMaterialAttributes=true
  bUsedWithClothing=true
  bUsedWithMeshDeformer=true
  bUsedWithMorphTargets=true
  bUsedWithNanite=true
  bUsedWithNiagaraMeshParticles=true
  bUsedWithSkeletalMesh=true
  pixelDepthOffsetMode=PDOM_Legacy
  previewMesh=/CosmeticShoes/Assets/FORT_Shoes/Shoe_JuggleFledWash/Meshes/JuggleFledWash_Shoe
  shadingModel=MSM_FromMaterialExpression
  usedShadingModels=Default Lit | Subsurface | Clear Coat | Subsurface Profile | Cloth
 }
 nodes [
  n1 = MaterialFunctionCall() {materialFunction=&MF_CharacterParent_2023_PreFX}
  n2 = MaterialFunctionCall!(Material=@n1.Result) {materialFunction=/Game/Global/Materials/Cosmetics/Kicks/Shoes/Materials/MF_ShoeReactivityTest}
  n3 = GetMaterialAttributes(None=MaterialFunctionCall!(In=@n2.Result, Pre Skinned Position=@n1.Pre skinned Local Position) {materialFunction=&/MF_JuiceMediumSawSweet}.Result) {attributeGetTypes [#1]}
  n4 = TextureSampleParameter2D() {
   parameterName=Gummy_TechArtMask
   samplerSource=SSM_Wrap_WorldGroupSettings
   samplerType=SAMPLERTYPE_Masks
   texture=/CosmeticShoes/Assets/FORT_Shoes/Shoe_YukonGladRusset/Skins/DraughtWhetRazz/Textures/T_DraughtWhetRazz_Shoe_FX
  }
  n5 = MaterialFunctionCall(Hue Shift Percentage=Multiply(A=@n4.G, B=ScalarParameter() {parameterName=ShoeLaces_HueShift}), Texture=@n3.Emissive Color) {materialFunction=&HueShift}
  n6 = Multiply(A=@n4.G, B=ScalarParameter() {parameterName=ShoeLaces_Desaturation}) {constB=-0.05}
  n7 = TextureSampleParameter2D() {
   parameterName=SweetShoelaceAOMult
   samplerType=SAMPLERTYPE_Masks
   texture=/CosmeticShoes/Assets/FORT_Shoes/Shoe_YukonGladRusset/Skins/DraughtWhetRazz/Textures/T_DraughtWhetRazz_Shoe_M
  }
  n8 = OneMinus(None=ScalarParameter() {defaultValue=0.8, parameterName=Sweet_AO_Strenght})
  n9 = Multiply(A=Multiply(A=Desaturation(None=@n5.Result, Fraction=@n6)) {constB=0.8}, B=Saturate(None=Add(A=@n7.R, B=@n8)))
  n10 = Constant() {r=1}
  n11 = Add(A=@n4.R, B=@n4.R)
  n12 = GetMaterialAttributes(None=BlendMaterialAttributes(
    A=SetMaterialAttributes(MaterialAttributes=@n3.MaterialAttributes, Emissive Color=@n9) {attributeSetTypes [#1]}
    B=MaterialFunctionCall!(
     SlurpMask=@n10
     AO
     BubblesMask
     SlurpNormalDistortionMask
     Pre Skinned Position=@n1.Pre skinned Local Position
     SecondColor_Mask=@n4.B
     Material=@n2.Result
    ) {materialFunction=&/MF_Slurp_JuiceMediumSaw}.Result
    Alpha=Saturate(None=@n11)
   )) {attributeGetTypes [#1]}
  n13 = SkyAtmosphereLightDirection()
  n14 = CameraVectorWS()
  n15 = Power(Base=ConstantBiasScale(None=DotProduct(A=@n13, B=@n14))) {constExponent=0.75}
  n16 = Multiply(
   A=LinearInterpolate(Alpha=Power(Base=@n1.Fresnel) {constExponent=3}) {constA=1, constB=2}
   B=LinearInterpolate(Alpha=OneMinus(None=@n15)) {constA=1, constB=2}
  )
  n17 = SetMaterialAttributes(MaterialAttributes=@n12.MaterialAttributes, Emissive Color=Multiply(A=Multiply(A=@n12.Emissive Color, B=@n16)) {constB=0.52}) {attributeSetTypes [#1]}
  MaterialFunctionCall!(In=MaterialFunctionCall(
    Material=MaterialFunctionCall!(InOpacity, Material=@n17, Mask=@n1.Tech Art Mask) {materialFunction=/Game/Global/Materials/Cosmetics/Kicks/Socks/Materials/MF_Sock}.Result
    M=@n1.M
    Pre Skinned Local Position=@n1.M
    Distance Blend=@n1.M
    Fresnel=@n1.M
   ) {materialFunction=&MF_CharacterParent_2023_PostFX}.Result) {materialFunction=/Game/Global/Materials/Cosmetics/Kicks/Shoes/Materials/MF_TechartShoeHologram}
 ]
}
